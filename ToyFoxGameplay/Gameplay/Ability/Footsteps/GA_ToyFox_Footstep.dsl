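; Event-triggered gameplay ability that deals footstep damage for the ToyFox giant player pawn.
; A targeting task replicates client hits to the server, which commits cooldown and sorts them into pawns, large or immune-tagged buildings, and small buildings.
; Throttled async effects, originating at the footstep, damage pawns then large buildings; small ones go to the GameState destruction manager for silent destruction.
; Failure reasons are blanked to hide cooldown errors.
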(event Ability|EventActivateAbilityFromEvent
  (params (EventData :type Gameplay%20Event%20Data%20Structure%20%28by%20ref%29))
  (bind _returnvalue (Ability|GetAbilitySystemComponentfromActorInfo))
  (bind _footstepworlddamagegameplayeffect (Variables|Settings|GetFootstepWorldDamageGameplayEffect))
  (bind _footsteppawndamagegameplayeffect (Variables|Settings|GetFootstepPawnDamageGameplayEffect))
  ;[0 0.15 0.0188 0.5] ActivateAbilityFromEvent: Only worry about damage here.
  (|CallParentFunction EventData)
  (bind (_asynctaskproxy _targetdata) (Ability|Tasks|PerformTargetingAbilityTask "DamageTargeting" (Variables|Targeting|GetFootstepTargetingPreset) false false true)
    (:Targeted
      _)
    (:Cancelled
      _)) ; This node needs to be used since it will replicate target\n data from client to server for best accuracy.
  (Ability|HasAuthority
    (:True
      (Ability|CommitAbility) ; Apply cooldown
      (bind (_pawnhitresults _cosmeticbuildingdestructionhitresults _silentbuildingdestructionhitresults) (CallFunction|ParseTargetDataHitResults _targetdata))
      ;;[9.8e-05 0.1234 0.15 0.5] Async damage pawns.
      (bind _returnvalue_2 (Ability|GameplayEffect|GetEffectContext (CallFunction|CreateEffectSpecforFootstepDamageApplication _footsteppawndamagegameplayeffect)))
      (Ability|Tasks|ApplyGameplayEffectToHitResultsAsync _returnvalue _pawnhitresults _footsteppawndamagegameplayeffect 1 1 _returnvalue_2 (Variables|Hotfix|GetHotfix_MaxTargetsToProcessPerTick)
        (:OnCompleted
          ;;[0.0871 0 0.15 0.5] Cosmetic destroy larger buildings by applying large damage.
          (bind _returnvalue_3 (Ability|GameplayEffect|GetEffectContext (CallFunction|CreateEffectSpecforFootstepDamageApplication _footstepworlddamagegameplayeffect)))
          (Ability|Tasks|ApplyGameplayEffectToHitResultsAsync _returnvalue _cosmeticbuildingdestructionhitresults _footstepworlddamagegameplayeffect 1 1 _returnvalue_3 (Variables|Hotfix|GetHotfix_MaxTargetsToProcessPerTick)
            (:OnCompleted
              ;;[0.1236 0.05255 0.15 0.5] Silent destory smaller buildings by sending it to the destruction component.
              (bind _returnvalue_4 (Fortnite|Athena|GameState|GetGameStateAthena))
              (bind _returnvalue_5 (Actor|GetComponentbyClass _returnvalue_4 "/GiantPawnCore/Gameplay/Destruction/B_GiantPawnDestructionManagerGameStateComponent.B_GiantPawnDestructionManagerGameStateComponent_C"))
              (Utilities|IsValid _returnvalue_5
                (:Is%20Valid
                  (Public|AddBuildingsforDestructionHitResults _returnvalue_5 _silentbuildingdestructionhitresults "FootstepDestruction" (CallFunction|GetAvatarasFortPlayerPawn)))
                (:Is%20Not%20Valid
                  (Fortnite|FortShippingLog self "ActivateAbility: GiantPawnDestructionComponent is nullptr." true)))
              (Ability|EndAbility))))))
    (:False
      (Ability|EndAbility) ; End client ability once it has done its targeting.
    )))

(fn CreateEffectSpecForFootstepDamageApplication
  (params (GameplayEffectClass :type Gameplay%20Effect%20Class%20Reference))
  (Utilities|FlowControl|Sequence
    (:then_0
      (bind _output_get (LocalVariables|SetEffectSpec (Ability|MakeOutgoingGameplayEffectSpec GameplayEffectClass)))
      (bind _returnvalue (Ability|GameplayEffect|GetEffectContext _output_get))
      (bind _asb_toy_fox_giant_player_pawn (Utilities|Casting|CastToB_ToyFoxGiantPlayerPawn (CallFunction|GetAvatarasFortPlayerPawn)))
      (Ability|EffectContext|SetOrigin _returnvalue (Footsteps|GetLatestFootstepHitLocation _asb_toy_fox_giant_player_pawn)))
    (:then_1
      (return (LocalVariables|GetEffectSpec)))))

(fn OverrideFailedReason
  (params (FailedReason :type Gameplay%20Tag%20Container%20Structure%20%28by%20ref%29))
  ; Override fail reason to be empty since we do not want to show any errors on the screen when this ability fails.
  ; Most prone to happening when ability is still in cooldown in case two footsteps happen within a very short timeframe.
  (return 0))

(fn MakeGameplayCueParamForFootstepImpact
  (params (HitResult :type Hit%20Result%20Structure%20%28by%20ref%29))
  (bind (_normalizedmagnitude _rawmagnitude _effectcontext _matchedtagname _originaltag _aggregatedsourcetags _aggregatedtargettags _location _normal _instigator _effectcauser _sourceobject _physicalmaterial _gameplayeffectlevel _abilitylevel _targetattachcomponent _breplicatelocationwhenusingminimalrepproxy) (Ability|GameplayCue|BreakGameplayCueParameters (GameplayCue|MakeGameplayCueParametersfromHitResult HitResult)))
  (bind _output_get (LocalVariables|SetEffectContext (Class|AthenaFunctionLibrary|MakeSimpleFortEffectContext (CallFunction|GetAvatarasFortPlayerPawn) (CallFunction|GetAvatarasFortPlayerPawn))))
  (Ability|EffectContext|AddHitResult _output_get HitResult)
  (return (Ability|GameplayCue|MakeGameplayCueParameters _normalizedmagnitude _rawmagnitude _output_get _matchedtagname _originaltag _aggregatedsourcetags _aggregatedtargettags _location _normal _instigator _effectcauser _sourceobject _physicalmaterial _gameplayeffectlevel _abilitylevel _targetattachcomponent _breplicatelocationwhenusingminimalrepproxy)))

(fn ParseTargetDataHitResults
  (params (TargetData :type Gameplay%20Ability%20Target%20Data%20Handle%20Structure))
  (Utilities|FlowControl|Sequence
    (:then_0
      (for _index (range (Ability|TargetData|GetDataCountfromTargetData TargetData))
        (bind _returnvalue (Ability|TargetData|GetHitResultfromTargetData TargetData _index))
        (Utilities|IsValid (.HitActor (Collision|BreakHitResult _returnvalue))
          (:Is%20Valid
            (bind _asbuilding_actor (Utilities|Casting|CastToBuildingActor (.HitActor (Collision|BreakHitResult _returnvalue)))
              (:then
                (if (Variables|/Script/Engine.Actor|GetbCanBeDamaged _asbuilding_actor)
                  (bind _returnvalue_2 (Math|Vector|VectorLengthSquared (.BoxExtent (Collision|GetActorBounds _asbuilding_actor true false))))
                  (if (or (> _returnvalue_2 (ScalableFloat|GetCachedScalableFloatValue (Variables|Hotfix|GetHotfix_BuildingActorVolumeSqToCosmeticDestroy))) (Actor|ActorHasTag (.HitActor (Collision|BreakHitResult _returnvalue)) "GiantDestructionImmune"))
                    (Utilities|Array|Add (LocalVariables|GetRetCosmeticDestructionHitResults) _returnvalue)
                    (if (Variables|Debug|GetbDebug)
                      (Fortnite|FortShippingLog self (Utilities|String|Append "ParseTargetDataHitResults: CosmeticDestruction: Volume of actor " (Utilities|GetDisplayName (.HitActor (Collision|BreakHitResult _returnvalue))) " is " (Utilities|String|ToString(Float) _returnvalue_2))))
                    (else
                      (Utilities|Array|Add (LocalVariables|GetRetSilentDestructionHitResults) _returnvalue)
                      (if (Variables|Debug|GetbDebug)
                        (Fortnite|FortShippingLog self (Utilities|String|Append "ParseTargetDataHitResults: SilentDestruction: Volume of actor " (Utilities|GetDisplayName (.HitActor (Collision|BreakHitResult _returnvalue))) " is " (Utilities|String|ToString(Float) _returnvalue_2))))))))
              (:CastFailed
                (Utilities|Array|Add (LocalVariables|GetRetPawnHitResults) _returnvalue)
                (if (Variables|Debug|GetbDebug)
                  (Fortnite|FortShippingLog self (Utilities|String|Append "ParseTargetDataHitResults: Adding actor (pawn): " (Utilities|GetDisplayName (.HitActor (Collision|BreakHitResult _returnvalue))))))))))))
    (:then_1
      (return (LocalVariables|GetRetPawnHitResults) (LocalVariables|GetRetCosmeticDestructionHitResults) (LocalVariables|GetRetSilentDestructionHitResults)))))
