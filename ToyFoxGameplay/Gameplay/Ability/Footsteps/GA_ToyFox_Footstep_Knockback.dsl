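; GA_ToyFox_Footstep_Knockback is a gameplay ability that supplies footstep knockback with its source point, using a parent class for the rest of the activation logic.
; On activation it caches the origin from the granting effect's context, then calls the parent implementation.
; Its GetKnockbackOrigin function returns that cached point and ignores the OriginActor parameter, so knockback originates from the effect's location rather than from an actor.
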(event Ability|EventActivateAbility
  (Utilities|FlowControl|Sequence
    (:then_0
      (Variables|Default|SetFootstepKnockbackOrigin (Ability|EffectContext|GetOrigin (Ability|GetGrantedbyEffectContext))))
    (:then_1
      (|CallParentFunction))))

(fn GetKnockbackOrigin
  (params (OriginActor :type Actor%20Object%20Reference))
  (return (Variables|Default|GetFootstepKnockbackOrigin)))
